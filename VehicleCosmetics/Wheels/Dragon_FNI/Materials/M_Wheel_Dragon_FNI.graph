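# Subsurface material for the Dragon FNI vehicle cosmetic wheel, covering rim and tire, with a Tire Mat static switch gating the tire look.
# The rim combines base color, masks and custom color with bump-offset normals and Fresnel rim light for fake depth and refraction.
# The tire layers flow-distorted flame emissive with parallax, optionally panning ice and a cloudy-sky reflection lookup.
# Named reroutes carry both branches into masked blends feeding material attributes.
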Material /VehicleCosmetics/Wheels/Dragon_FNI/Materials/M_Wheel_Dragon_FNI
graph {
 properties {
  bAutomaticallySetUsageInEditor=false
  bUseMaterialAttributes=true
  bUsedWithMeshDeformer=true
  bUsedWithSkeletalMesh=true
  previewMesh=/VehicleCosmetics/Wheels/Dragon_FNI/Meshes/WHEEL_Dragon_SK_iMats
  shadingModel=MSM_Subsurface
  usedShadingModels=Subsurface
 }
 nodes [
  n1 = Add(A=NamedRerouteUsage(), B=NamedRerouteUsage())
  n2 = NamedRerouteUsage()
  n3 = VectorParameter() {group=00 Base, parameterName=Normal Transform, sortPriority=-2, defaultValue=<rgba 1 1 0 0>}
  n4 = Add(A=Multiply(A=TextureCoordinate(), B=AppendVector(A=@n3.R, B=@n3.R)), B=AppendVector(A=@n3.B, B=@n3.B))
  n5 = TextureObjectParameter(UVs=@n4) {
   group=00 Base
   parameterName=Rim Normal
   samplerType=SAMPLERTYPE_Normal
   sortPriority=1
   texture=/VehicleCosmetics/Wheels/Dragon_FNI/Textures/T_Dragon_Fni_Rim_N
  }
  n6 = TextureSample(Tex=@n5) {
   samplerSource=SSM_Wrap_WorldGroupSettings
   samplerType=SAMPLERTYPE_Normal
   texture=&T_Dragon_Fni_Rim_N
  }
  n7 = Add(A=@n6.R, B=@n6.G)
  n8 = ScalarParameter() {defaultValue=1, group=01 Rim, parameterName=RIM Normal Intensity, sortPriority=0}
  n9 = Clamp(None=Multiply(A=@n7, B=@n8))
  n10 = ConstantBiasScale(None=Power(Base=NamedRerouteUsage(), Exp=LinearInterpolate(Alpha=@n9) {constA=8, constB=3})) {bias=-0.025, scale=1}
  n11 = Add(
   A=Clamp(None=ConstantBiasScale(None=@n1) {bias=-0.1, scale=3}) {maxDefault=3, minDefault=-0.1}
   B=Multiply(A=ConstantBiasScale(None=@n2) {bias=-0.1, scale=3}, B=Clamp(None=@n10) {maxDefault=3, minDefault=-0.1})
  )
  Clamp(None=ConstantBiasScale(None=@n11) {bias=0.75, scale=1})
  n12 = NamedRerouteUsage()
  n13 = Multiply(A=@n12) {constB=0.5}
  Multiply(A=Desaturation(None=@n13, Fraction=Constant() {r=0.6}) {luminanceFactors=<rgba 0.3 0.59 0.11 0>}) {constB=0.9}
  n14 = VectorParameter() {group=00 Base, parameterName=Custom Color, sortPriority=-3, defaultValue=<rgba 0.2051 0.2542 0.8796 1>}
  NamedRerouteDeclaration(None=@n14.RGB) {name=Custom Color, nodeColor=<rgba 1 0.5 0 1>}
  n15 = NamedRerouteUsage()
  n16 = TextureCoordinate() {uTiling=0.5, vTiling=0.3}
  n17 = ScalarParameter() {defaultValue=1, group=03 Tire Emissive, parameterName=Tire UV Distort Tile, sortPriority=1}
  n18 = TextureSample(UVs=Panner(Coordinate=Multiply(A=@n16, B=@n17)) {speedX=0.1, speedY=0.125}) {
   samplerSource=SSM_Wrap_WorldGroupSettings
   samplerType=SAMPLERTYPE_Normal
   texture=/VehicleCosmetics/Wheels/_SharedTextures/FX/Flow_03_nomips
  }
  n19 = ComponentMask(None=@n18.RGB) {g=true, r=true}
  n20 = ScalarParameter() {defaultValue=0.25, group=03 Tire Emissive, parameterName=Tire UV Distort Amt, sortPriority=2}
  n21 = TextureObjectParameter() {group=03 Tire Emissive, parameterName=Tire Flame, sortPriority=0, texture=/VehicleCosmetics/Wheels/_SharedTextures/FX/FlameTire_M_T}
  n22 = TextureSample(UVs=Add(A=TextureCoordinate(), B=Multiply(A=@n19, B=@n20)), Tex=@n21) {samplerSource=SSM_Wrap_WorldGroupSettings, texture=&FlameTire_M_T}
  n23 = Multiply(A=@n15, B=@n22.R)
  n24 = TextureObjectParameter() {group=00 Base, parameterName=Rim Masks, sortPriority=2, texture=/VehicleCosmetics/Wheels/Dragon_FNI/Textures/T_Dragon_Fni_M}
  n25 = TextureSample(Tex=NamedRerouteDeclaration(None=@n24) {name=Rim Mask Tex, nodeColor=<rgba 0.7578 1 0 1>}) {samplerSource=SSM_Wrap_WorldGroupSettings, texture=&FlameTire_M_T}
  n26 = Multiply(A=Multiply(A=@n23) {constB=5}, B=@n25.R)
  n27 = Constant() {r=1}
  n28 = Constant()
  n29 = NamedRerouteUsage()
  n30 = TextureCoordinate() {uTiling=0.5}
  n31 = TextureSample(UVs=@n30, Tex=@n21) {samplerSource=SSM_Wrap_WorldGroupSettings, texture=&FlameTire_M_T}
  n32 = Add(A=@n31.R, B=@n31.R)
  n33 = OneMinus(None=@n32)
  n34 = Multiply(A=Multiply(A=Clamp(None=@n33), B=@n31.B), B=@n25.B)
  n35 = Add(A=@n23, B=Multiply(A=Multiply(A=@n29, B=@n34)) {constB=10})
  n36 = Multiply(
   A=@n15
   B=TextureSample(UVs=Panner(Coordinate=@n30) {speedX=0.05}, Tex=@n21) {samplerSource=SSM_Wrap_WorldGroupSettings, texture=&FlameTire_M_T}.R
  )
  n37 = Clamp(None=Subtract(A=Add(A=@n25.B, B=@n31.B), B=@n32))
  n38 = Multiply(A=Multiply(A=Multiply(A=@n36) {constB=3}, B=Multiply(A=@n22.B, B=@n37))) {constB=10}
  NamedRerouteDeclaration(None=LinearInterpolate(
    A=Multiply(A=@n26) {constB=2}
    B=Add(A=@n35, B=@n38)
    Alpha=StaticSwitchParameter(True=@n27, False=@n28) {group=00 Base, parameterName=Tire Mat, sortPriority=-5}
   )) {name=Tire Emissive, nodeColor=<rgba 1 0 0.1641 1>}
  n39 = TextureObjectParameter() {group=02 Tire Ice, parameterName=Ice Masks, sortPriority=0, texture=/VehicleCosmetics/Wheels/_SharedTextures/FX/T_IceRedux_2RGB}
  n40 = TextureSample(UVs=NamedRerouteUsage(), Tex=@n39) {samplerSource=SSM_Wrap_WorldGroupSettings, texture=&T_IceRedux_2RGB}
  NamedRerouteDeclaration(None=@n40.R) {name=Ice Masks R, nodeColor=<rgba 1 0 0.1875 1>}
  NamedRerouteDeclaration(None=@n40.G) {name=Ice Masks G, nodeColor=<rgba 1 0 0.1875 1>}
  n41 = Desaturation(None=NamedRerouteUsage()) {luminanceFactors=<rgba 0.3 0.59 0.11 0>}
  NamedRerouteDeclaration(None=Clamp(None=ConstantBiasScale(None=@n41) {bias=0, scale=15})) {name=Mask for Tire Lerps, nodeColor=<rgba 0 1 0.5781 1>}
  n42 = NamedRerouteUsage()
  n43 = NamedRerouteUsage()
  n44 = Multiply(A=ConstantBiasScale(None=OneMinus(None=@n43)) {bias=-0.25, scale=2}, B=NamedRerouteUsage())
  n45 = Subtract(A=@n44, B=@n42)
  n46 = Add(A=ConstantBiasScale(None=Clamp(None=@n45)) {bias=0}, B=ConstantBiasScale(None=@n42) {bias=0})
  n47 = Multiply(A=NamedRerouteUsage(), B=NamedRerouteUsage())
  n48 = Desaturation(None=Add(A=Clamp(None=@n47), B=NamedRerouteUsage()), Fraction=Constant() {r=0.5}) {luminanceFactors=<rgba 0.3 0.59 0.11 0>}
  NamedRerouteDeclaration(None=LinearInterpolate(A=Multiply(A=@n46, B=@n48), Alpha=NamedRerouteUsage()) {constB=0}) {name=Tire Color, nodeColor=<rgba 1 0.2578 0 1>}
  n49 = Multiply(A=ComponentMask(None=ReflectionVectorWS()) {g=true, r=true}) {constB=0.5}
  n50 = NamedRerouteUsage()
  n51 = ConstantBiasScale(None=NamedRerouteUsage()) {bias=0.5}
  n52 = TextureSampleParameter2D(UVs=Add(A=Add(A=@n49) {constB=0.5}, B=Multiply(A=@n50, B=@n51))) {
   group=02 Tire Ice
   parameterName=Ice Reflection map
   samplerSource=SSM_Wrap_WorldGroupSettings
   sortPriority=3
   texture=/VehicleCosmetics/Wheels/_SharedTextures/FX/_2D_ENV_CloudySky
  }
  n53 = Multiply(
   A=@n52.RGB
   B=MaterialFunctionCall!(Normal, Exponent=Constant() {r=1}) {materialFunction=/VehicleCosmetics/Wheels/Cola_Holographic/Materials/MF_Fresnel_Classic}.Result
  )
  n54 = MaterialFunctionCall(Normal=NamedRerouteUsage(), Exponent=Constant() {r=2}) {materialFunction=&MF_Fresnel_Classic}
  n55 = NamedRerouteUsage()
  n56 = NamedRerouteUsage()
  n57 = Multiply(A=Multiply(A=LinearInterpolate(Alpha=@n54.Result) {constA=3, constB=0}, B=Multiply(A=@n55, B=@n56)), B=@n55)
  n58 = ConstantBiasScale(None=@n43) {bias=0.5}
  n59 = LinearInterpolate(
   A=@n58
   Alpha=MaterialFunctionCall(Exponent=Constant() {r=2}) {materialFunction=&MF_Fresnel_Classic}.Result
  )
  n60 = Add(A=Multiply(A=NamedRerouteUsage(), B=NamedRerouteUsage()), B=Add(A=Add(A=@n58, B=@n59), B=Clamp(None=@n44)))
  n61 = Power(
   Base=MaterialFunctionCall(Normal=NamedRerouteUsage(), Exponent=Constant() {r=2}) {materialFunction=&MF_Fresnel_Classic}.Result
   Exp=ScalarParameter() {defaultValue=4, group=01 RimLight, parameterName=Rim Light Sharpness}
  )
  n62 = Multiply(A=Add(A=Add(A=@n57, B=@n60), B=Multiply(A=@n61) {constB=2}), B=@n48)
  n63 = ConstantBiasScale(None=NamedRerouteUsage()) {scale=0.6}
  n64 = LinearInterpolate(A=Multiply(A=Add(A=@n53, B=@n62), B=Clamp(None=@n63)), Alpha=NamedRerouteUsage()) {constB=0}
  n65 = NamedRerouteUsage()
  NamedRerouteDeclaration(None=Multiply(A=Add(A=@n64, B=@n65)) {constB=0.5}) {name=Tire Emissive Final, nodeColor=<rgba 1 0.2578 0 1>}
  n66 = NamedRerouteUsage()
  n67 = Multiply(
   A=ComponentMask(None=@n66) {g=true, r=true}
   B=ScalarParameter() {defaultValue=2, group=02 Tire Ice, parameterName=Ice Normal Intensity, sortPriority=5}
  )
  n68 = ComponentMask(None=@n66) {b=true}
  NamedRerouteDeclaration(None=LinearInterpolate(A=AppendVector(A=@n67, B=@n68), B=Constant3Vector() {constant=<rgba 0 0 1 0>}, Alpha=NamedRerouteUsage())) {name=Tire Normal, nodeColor=<rgba 0.5 0.5 1 1>}
  n69 = Multiply(A=ComponentMask(None=@n6.RGB) {g=true, r=true}, B=@n8)
  NamedRerouteDeclaration(None=AppendVector(A=@n69, B=@n6.B)) {name=Rim Normal, nodeColor=<rgba 0 1 1 1>}
  n70 = TextureSample(
   UVs=BumpOffset(Coordinate=@n4, Height=ScalarParameter() {defaultValue=-0.5, group=01 Rim, parameterName=RIM Normal Bump Offset Amount, sortPriority=1})
   Tex=@n5
  ) {
   samplerSource=SSM_Wrap_WorldGroupSettings
   samplerType=SAMPLERTYPE_Normal
   texture=&T_Dragon_Fni_Rim_N
  }
  n71 = ScalarParameter() {defaultValue=0.5, group=01 Rim, parameterName=RIM Offset Reflections Multilpier, sortPriority=2}
  n72 = Multiply(
   A=ComponentMask(None=NamedRerouteDeclaration(None=@n70.RGB) {name=Rim Normal Offset, nodeColor=<rgba 0.4297 0 1 1>}) {g=true, r=true}
   B=@n71
  )
  NamedRerouteDeclaration(None=AppendVector(A=Add(A=@n69, B=@n72), B=@n70.B)) {name=Normal to Rim Fresnel, nodeColor=<rgba 0 0.7656 1 1>}
  n73 = CameraVectorWS()
  n74 = ComponentMask(None=Power(Base=Transform(None=@n73) {transformSourceType=TRANSFORMSOURCE_World, transformType=TRANSFORM_Tangent}) {constExponent=0.1}) {b=true}
  n75 = OneMinus(None=@n74)
  n76 = NamedRerouteUsage()
  n77 = Constant() {r=2}
  n78 = Multiply(
   A=Power(
    Base=MaterialFunctionCall(Normal=@n76, Exponent=@n77) {materialFunction=&MF_Fresnel_Classic}.Result
    Exp=ScalarParameter() {defaultValue=4, group=01 RimLight, parameterName=Rim Light Sharpness}
   ) {constExponent=0.1}
   B=ScalarParameter() {defaultValue=3, desc=RimLight and Normal Refraction effect amount, group=01 RimLight, parameterName=Rim Light Refract}
  )
  n79 = Clamp(None=LinearInterpolate(A=Multiply(A=@n75, B=@n78), B=Multiply(A=@n74, B=@n78)) {constAlpha=0.65})
  n80 = TextureSampleParameter2D() {
   group=00 Base
   parameterName=Rim Base Color
   samplerSource=SSM_Wrap_WorldGroupSettings
   sortPriority=0
   texture=/VehicleCosmetics/Wheels/Dragon_FNI/Textures/T_Dragon_Fni_Rim_BC
  }
  n81 = Desaturation(None=@n80.RGB) {luminanceFactors=<rgba 0.3 0.59 0.11 0>}
  n82 = Multiply(A=Clamp(None=ConstantBiasScale(None=@n81) {bias=0.2, scale=1}), B=NamedRerouteUsage())
  n83 = Multiply(
   A=@n82
   B=ScalarParameter() {defaultValue=1, desc=RimLight and Normal Refraction effect Intensity, group=01 RimLight, parameterName=Rim Light Power}
  )
  NamedRerouteDeclaration(None=Multiply(A=Multiply(A=@n79, B=@n83), B=@n9)) {name=Rim Light, nodeColor=<rgba 0 1 0.6953 1>}
  n84 = Add(A=@n7, B=Multiply(A=Add(A=@n70.R, B=@n70.R), B=@n71))
  NamedRerouteDeclaration(None=Clamp(None=Multiply(A=@n8, B=@n84))) {name=Rim Offset Reflections, nodeColor=<rgba 0 1 0.2031 1>}
  NamedRerouteDeclaration(None=@n25.R) {name=Rim Masks R, nodeColor=<rgba 0.2656 1 0 1>}
  NamedRerouteDeclaration(None=@n82) {name=BC GrayScale * CC, nodeColor=<rgba 0 0.9297 1 1>}
  n85 = NamedRerouteUsage()
  n86 = TextureSample(UVs=BumpOffset(Coordinate=@n4, Height=Constant()) {heightRatio=-0.1}, Tex=@n85) {samplerSource=SSM_Wrap_WorldGroupSettings, texture=&FlameTire_M_T}
  n87 = TextureSample(Tex=@n85) {samplerSource=SSM_Wrap_WorldGroupSettings, texture=&FlameTire_M_T}
  NamedRerouteDeclaration(None=Clamp(None=Min(A=@n86.G, B=@n87.G))) {name=AO, nodeColor=<rgba 1 0.75 0 1>}
  n88 = ScalarParameter() {group=01 Rim Colors, parameterName=RIM AO Intensity, sortPriority=3}
  n89 = ScalarParameter() {defaultValue=4, group=01 Rim Colors, parameterName=RIM Ice Flatness, sortPriority=4}
  n90 = Multiply(A=NamedRerouteUsage(), B=@n89)
  n91 = Clamp(None=@n90)
  n92 = Clamp(None=Multiply(A=LinearInterpolate(A=@n89, B=@n90, Alpha=@n88), B=LinearInterpolate(A=@n13, B=@n12, Alpha=@n91)))
  n93 = ScalarParameter() {defaultValue=0.25, group=01 Rim Colors, parameterName=RIM Highlight Desaturation, sortPriority=5}
  n94 = Desaturation(None=@n92, Fraction=@n93) {luminanceFactors=<rgba 0.3 0.59 0.11 0>}
  n95 = Transform(None=CameraVectorWS()) {transformSourceType=TRANSFORMSOURCE_World, transformType=TRANSFORM_Tangent}
  n96 = ComponentMask(None=@n95) {b=true}
  n97 = Power(Base=@n96)
  n98 = NamedRerouteUsage()
  n99 = LinearInterpolate(A=Clamp(None=Power(Base=@n96) {constExponent=10}), B=Clamp(None=ConstantBiasScale(None=@n97) {bias=0, scale=5}), Alpha=@n98)
  n100 = Clamp(None=LinearInterpolate(A=@n92, B=@n94, Alpha=@n99))
  n101 = LinearInterpolate(A=Multiply(A=@n13) {constB=0.05}, B=Desaturation(None=@n13, Fraction=@n93) {luminanceFactors=<rgba 0.3 0.59 0.11 0>}, Alpha=@n99)
  n102 = ComponentMask(None=OneMinus(None=@n95)) {b=true}
  n103 = ScalarParameter() {defaultValue=4, group=01 RimLight, parameterName=Rim Light Sharpness}
  n104 = Power(Base=@n102, Exp=@n103)
  n105 = ScalarParameter() {defaultValue=1, desc=RimLight and Normal Refraction effect Intensity, group=01 RimLight, parameterName=Rim Light Power}
  n106 = Multiply(A=@n105, B=Power(Base=@n102, Exp=Multiply(A=@n103) {constB=0.5}))
  n107 = ScalarParameter() {defaultValue=9, group=01 Rim Colors, parameterName=RIM Inner Color Falloff, sortPriority=1}
  n108 = Multiply(
   A=Power(Base=@n96, Exp=Multiply(A=@n107) {constB=2})
   B=ScalarParameter() {defaultValue=0.25, group=01 Rim Colors, parameterName=RIM Inner Color Intensity, sortPriority=0}
  )
  n109 = TextureSampleParameter2D() {
   group=00 Base
   parameterName=Rim Masks SE
   samplerType=SAMPLERTYPE_LinearColor
   texture=/VehicleCosmetics/Wheels/Dragon_FNI/Textures/T_Dragon_Fni_Rim_M
  }
  n110 = LinearInterpolate(A=@n108, B=LinearInterpolate(A=Multiply(A=@n105, B=@n104), B=Max(A=@n106, B=@n108), Alpha=@n98), Alpha=@n109.R)
  n111 = ComponentMask(None=TextureCoordinate()) {r=true}
  n112 = Clamp(None=ConstantBiasScale(None=OneMinus(None=@n111)) {bias=-0.97, scale=5000})
  n113 = Constant() {r=1}
  n114 = Constant()
  n115 = LinearInterpolate(
   A=Clamp(None=ConstantBiasScale(None=@n110) {bias=0.1, scale=1})
   Alpha=LinearInterpolate(
    A=Multiply(A=@n109.B, B=@n112)
    Alpha=StaticSwitchParameter(True=@n113, False=@n114) {group=00 Base, parameterName=Tire Mat, sortPriority=-5}
   ) {constB=0}
  ) {constB=0}
  n116 = ConstantBiasScale(None=@n94) {bias=0.1, scale=1}
  n117 = MaterialFunctionCall(Normal=NamedRerouteUsage(), Exponent=Constant() {r=0.1}) {materialFunction=&MF_Fresnel_Classic}
  n118 = Clamp(None=ConstantBiasScale(None=OneMinus(None=@n117.Result)) {bias=-0.1, scale=1})
  n119 = ScalarParameter() {defaultValue=0.25, group=01 Rim Colors, parameterName=RIM Inner Color Intensity, sortPriority=0}
  NamedRerouteDeclaration(None=LinearInterpolate(A=LinearInterpolate(A=@n101, B=@n100, Alpha=@n115), B=Clamp(None=@n116), Alpha=Multiply(A=@n118, B=@n119))) {name=Rim Colors, nodeColor=<rgba 0 0.625 1 1>}
  NamedRerouteDeclaration(None=@n102) {name=Camera Vector B 1-x, nodeColor=<rgba 0.6172 0 1 1>}
  n120 = Constant() {r=1}
  n121 = Constant()
  NamedRerouteDeclaration(None=LinearInterpolate(A=NamedRerouteUsage(), Alpha=StaticSwitchParameter(True=@n120, False=@n121) {group=00 Base, parameterName=Tire Mat, sortPriority=-5})) {name=Main Lerp Mask, nodeColor=<rgba 1 0.3984 0 1>}
  NamedRerouteDeclaration(None=Clamp(None=@n1)) {name=Rim Final BC, nodeColor=<rgba 0.0781 0 1 1>}
  NamedRerouteDeclaration(None=Multiply(A=ScalarParameter() {defaultValue=1, group=01 Rim Colors, parameterName=RIM SSS Intensity}, B=@n11)) {name=SSS, nodeColor=<rgba 0.2422 1 0 1>}
  n122 = Multiply(
   A=TextureCoordinate() {uTiling=0.5, vTiling=0.25}
   B=ScalarParameter() {defaultValue=1.5, group=02 Tire Ice, parameterName=Ice Tiling, sortPriority=4}
  )
  NamedRerouteDeclaration(None=@n122) {name=Ice Tiling, nodeColor=<rgba 0.4531 0 1 1>}
  NamedRerouteDeclaration(None=Multiply(A=@n122) {constB=2}) {name=Ice Tiling x2, nodeColor=<rgba 0.4531 0 1 1>}
  n123 = TextureSampleParameter2D(UVs=NamedRerouteUsage()) {
   group=02 Tire Ice
   parameterName=Ice Normal
   samplerType=SAMPLERTYPE_Normal
   sortPriority=1
   texture=/VehicleCosmetics/Wheels/_SharedTextures/FX/T_IceRedux_1
  }
  NamedRerouteDeclaration(None=@n123.RGB) {name=Ice Normal, nodeColor=<rgba 0.5 0.5 1 1>}
  n124 = TextureSampleParameter2D(UVs=NamedRerouteUsage()) {
   group=02 Tire Ice
   parameterName=Ice Masks Additional
   samplerSource=SSM_Wrap_WorldGroupSettings
   sortPriority=2
   texture=/VehicleCosmetics/Wheels/_SharedTextures/FX/T_IceRedux_Additional
  }
  NamedRerouteDeclaration(None=@n124.R) {name=Ice GrayScale Islands, nodeColor=<rgba 1 0 0.1875 1>}
  NamedRerouteDeclaration(None=@n124.G) {name=Ice Additional G, nodeColor=<rgba 1 0 0.1875 1>}
  n125 = BumpOffset(Coordinate=NamedRerouteUsage(), Height=Constant() {r=-3})
  n126 = Panner(Coordinate=@n125) {speedX=0.1}
  NamedRerouteDeclaration(None=TextureSample(UVs=StaticSwitchParameter(True=@n126, False=@n125) {defaultValue=true, group=02 Tire Ice, parameterName=Panning BG}, Tex=@n39) {samplerSource=SSM_Wrap_WorldGroupSettings, texture=&T_IceRedux_2RGB}.R) {name=Ice RGB Panner R, nodeColor=<rgba 1 0 0.1875 1>}
  n127 = NamedRerouteUsage()
  n128 = Constant() {r=-1}
  NamedRerouteDeclaration(None=TextureSample(UVs=BumpOffset(Coordinate=@n127, Height=@n128), Tex=@n39) {samplerSource=SSM_Wrap_WorldGroupSettings, texture=&T_IceRedux_2RGB}.B) {name=Ice Masks Bump B, nodeColor=<rgba 1 0 0.1875 1>}
  n129 = NamedRerouteUsage()
  SetMaterialAttributes!(
   MaterialAttributes
   Base Color=LinearInterpolate(A=NamedRerouteUsage(), B=NamedRerouteUsage(), Alpha=NamedRerouteUsage())
   Metallic
   Roughness
   Emissive Color=LinearInterpolate(B=NamedRerouteUsage(), Alpha=NamedRerouteUsage())
   Normal=LinearInterpolate(A=NamedRerouteUsage(), B=NamedRerouteUsage(), Alpha=NamedRerouteUsage())
   Subsurface Color=@n129
   Opacity=@n129
  ) {attributeSetTypes [
    #1
    #2
    #3
    #4
    #5
    #6
    #7
   ]}
 ]
}
