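# Translucent unlit particle material for a smoke effect, named as freeze smoke, usable on Niagara sprites, ribbons, meshes and skeletal meshes.
# A panning cloud texture slightly distorts the UVs of a flame-cloud texture.
# Two panning noise textures, one contrast-stretched, mask that cloud into boosted highlights added back onto it.
# Particle color tints the result with a dynamic Brightness parameter, while particle alpha scales the same result separately.
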Material /VehicleCosmetics/Wheels/Dragon_FNI/VFX/M_FreezeSmoke
graph {
 properties {
  bAutomaticallySetUsageInEditor=false
  bUsedWithNiagaraMeshParticles=true
  bUsedWithNiagaraRibbons=true
  bUsedWithNiagaraSprites=true
  bUsedWithSkeletalMesh=true
  blendMode=BLEND_Translucent
  previewMesh=/VehicleCosmetics/Wheels/SR_14_T1/Meshes/SK_Wheel_SR_14_T1
  shadingModel=MSM_Unlit
  usedShadingModels=Unlit
 }
 nodes [
  n1 = ParticleColor()
  n2 = TextureCoordinate() {uTiling=0.25, vTiling=0.25}
  n3 = Multiply(A=Time(), B=Constant() {r=0.5})
  n4 = ComponentMask(None=TextureSample(UVs=Panner(Coordinate=@n2, Time=@n3) {speedX=0.1, speedY=0.1}) {texture=&/Clouds_RGB_Soft_T}.RGB) {g=true, r=true}
  n5 = Constant() {r=0.1}
  n6 = TextureSample(UVs=Add(A=Multiply(A=@n4, B=@n5), B=TextureCoordinate())) {texture=/VehicleCosmetics/Wheels/SnowFlake/VFX/FlameCloud_01}
  n7 = Time()
  n8 = Constant() {r=2}
  n9 = TextureSample(UVs=Panner(Coordinate=TextureCoordinate() {uTiling=0.125, vTiling=0.125}, Time=Multiply(A=@n7, B=@n8)) {speedX=1, speedY=0.0001}) {texture=/VehicleCosmetics/Wheels/Tondi_Inverted/Textures/Universal_Noise_42_RGB}
  n10 = Panner(Coordinate=TextureCoordinate() {uTiling=1.5, vTiling=1.5}) {speedY=0.2}
  n11 = Multiply(
   A=Multiply(
    A=ConstantBiasScale(None=@n9.R) {bias=-0.5, scale=5}
    B=TextureSample(UVs=@n10) {texture=/VehicleCosmetics/Wheels/_SharedTextures/FX/Universal_Noise_32_RGB}.R
   )
   B=@n6.RGB
  )
  n12 = Add(A=Clamp(None=@n6.RGB), B=Multiply(A=Clamp(None=@n11), B=Constant() {r=10}))
  Multiply(A=Multiply(A=@n1.RGB, B=@n12), B=DynamicParameter() {paramNames [Param1,Brightness,Param3,Param4]}.Brightness)
  Multiply(A=@n12, B=@n1.A)
 ]
}
